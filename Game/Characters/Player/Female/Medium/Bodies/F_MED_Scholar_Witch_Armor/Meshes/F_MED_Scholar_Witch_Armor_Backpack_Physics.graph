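# Physics asset for the backpack mesh of the female medium Scholar Witch Armor, setting up dynamic cape bones.
# Box bodies form left, middle and right chains from small sphere bases, with tight angular limits near the top and looser swing toward the tips.
# Free-angular constraints tie side chains to the middle at lower segments, with the right-to-middle 03 link duplicated.
# Capsules cover pelvis, upper spine, legs, upper arms and right forearm.
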PhysicsAsset /Game/Characters/Player/Female/Medium/Bodies/F_MED_Scholar_Witch_Armor/Meshes/F_MED_Scholar_Witch_Armor_Backpack_Physics
bodies {
 calf_l [{shapeType=Capsule, length=23.6295, radius=6.3706, center=<-18.8778 -4.9273 -0.0001>, rotation=<rot 90 0 0>}]
 calf_r [{shapeType=Capsule, length=22.5306, radius=7.4694, center=<16.2116 3.8222 0>, rotation=<rot 90 0 0>}]
 dyn_cape_left_01 [{shapeType=Box, extentX=14.2411, extentY=17.4695, extentZ=2.9322, center=<7.131 0.0822 -0.1298>, rotation=<rot -8.7716 5.5926 0>}]
 dyn_cape_left_02 [{shapeType=Box, extentX=15.3641, extentY=20.0912, extentZ=4.8916, center=<7.5526 0 0.0001>}]
 dyn_cape_left_03 [{shapeType=Box, extentX=14.1779, extentY=20.4516, extentZ=6.6084, center=<7.172 0 0.0002>, rotation=<rot 0 0 -3.8166>}]
 dyn_cape_left_04 [{shapeType=Box, extentX=11.6101, extentY=21.0014, extentZ=7.479, center=<5.7996 0 -0.0001>, rotation=<rot 0 0 -1.9165>}]
 dyn_cape_left_05 [{shapeType=Box, extentX=12.5143, extentY=21.0001, extentZ=6.194, center=<6.4453 0 -0.0004>}]
 dyn_cape_left_06 [{shapeType=Box, extentX=11.8703, extentY=14.6584, extentZ=4.3381, center=<5.383 0 0.0002>}]
 dyn_cape_left_07 [{shapeType=Box, extentX=11.4537, extentY=11.987, extentZ=4, center=<5.0215 0 -0.0001>}]
 dyn_cape_left_base [{shapeType=Sphere, radius=0.5}]
 dyn_cape_left_parent [{shapeType=Box, extentX=13.8249, extentY=16.3268, extentZ=3.8501, center=<6.8642 0 2.4984>, rotation=<rot 23.1946 0 0>}]
 dyn_cape_mid_01 [{shapeType=Box, extentX=14.813, extentY=18.254, extentZ=4.4042, center=<7.5945 0 0.0003>}]
 dyn_cape_mid_02 [{shapeType=Box, extentX=14.813, extentY=17.8346, extentZ=4.7629, center=<7.3953 0 0.0002>}]
 dyn_cape_mid_03 [{shapeType=Box, extentX=14.813, extentY=16.3689, extentZ=4.7231, center=<7.1869 0 0.0002>}]
 dyn_cape_mid_04 [{shapeType=Box, extentX=14.813, extentY=17.2502, extentZ=4.2152, center=<7.2194 0 0.0004>}]
 dyn_cape_mid_05 [{shapeType=Box, extentX=19.0036, extentY=15.9556, extentZ=4.3304, center=<9.1916 -0.0001 0.0006>}]
 dyn_cape_mid_base [{shapeType=Sphere, radius=0.5}]
 dyn_cape_mid_parent [{shapeType=Box, extentX=14.813, extentY=11.5496, extentZ=3.0244, center=<8.196 0 0.0003>}]
 dyn_cape_right_01 [{shapeType=Box, extentX=14.3161, extentY=16.8224, extentZ=5.6932, center=<6.9436 0 0.0001>}]
 dyn_cape_right_02 [{shapeType=Box, extentX=14.7773, extentY=19.4333, extentZ=4.3037, center=<7.441 0 0.0002>}]
 dyn_cape_right_03 [{shapeType=Box, extentX=10.9024, extentY=15.6983, extentZ=5.8972, center=<5.4669 0 0.0001>}]
 dyn_cape_right_04 [{shapeType=Box, extentX=13.3225, extentY=15.0746, extentZ=4.5781, center=<6.2648 0 0>}]
 dyn_cape_right_base [{shapeType=Sphere, radius=0.5}]
 dyn_cape_right_parent [{shapeType=Box, extentX=14.6517, extentY=12.1165, extentZ=3.3632, center=<7.7071 0 0>}]
 lowerarm_r [{shapeType=Capsule, length=28.2542, radius=5.1861, center=<-10.9439 0.0001 0>, rotation=<rot 90 0 0>}]
 pelvis [{shapeType=Capsule, length=11.5267, radius=12.3483, center=<-4.1908 0.8487 0.0001>}]
 spine_05 [
  {shapeType=Capsule, length=19, radius=11, center=<-21.3959 4.1385 -0.0802>, rotation=<rot 88.0519 -0.0001 -11.463>}
  {shapeType=Capsule, length=0.9115, radius=4.2691, center=<-24.3248 -2.8704 -13.0096>, rotation=<rot 65.5611 -180 -180>}
  {shapeType=Capsule, length=12.516, radius=10.6833, center=<-2.2673 2.1233 0.0001>}
 ]
 thigh_l [{shapeType=Capsule, length=20.7062, radius=9.2938, center=<-27.1697 -0.0002 0>, rotation=<rot 90 0 0>}]
 thigh_r [
  {shapeType=Capsule, length=20.7062, radius=9.2938, center=<26.2158 0.0001 0>, rotation=<rot 90 0 0>}
  {shapeType=Capsule, length=8.9509, radius=4.8216, center=<-0.2443 1.9003 6.517>, rotation=<rot 77.6133 180 173.4074>}
 ]
 upperarm_l [{shapeType=Capsule, length=24.655, radius=5.345, center=<15.3534 -0.0003 0>, rotation=<rot 90 0 0>}]
 upperarm_r [{shapeType=Capsule, length=19.4724, radius=5, center=<-13.6692 0.8165 0.0416>, rotation=<rot -88.5465 41.0478 -36.1069>}]
}
constraints [
 {bone1Name=dyn_cape_mid_parent, bone2Name=dyn_cape_mid_base, swing1LimitDegrees=10, swing2LimitDegrees=10, twistLimitDegrees=5}
 {bone1Name=dyn_cape_mid_01, bone2Name=dyn_cape_mid_parent, swing1LimitDegrees=5, swing2LimitDegrees=5, twistLimitDegrees=30}
 {bone1Name=dyn_cape_mid_02, bone2Name=dyn_cape_mid_01, swing1LimitDegrees=5, swing2LimitDegrees=5, twistLimitDegrees=3}
 {bone1Name=dyn_cape_mid_03, bone2Name=dyn_cape_mid_02, swing1LimitDegrees=5, swing2LimitDegrees=5, twistLimitDegrees=3}
 {bone1Name=dyn_cape_mid_04, bone2Name=dyn_cape_mid_03, swing1LimitDegrees=45, swing2LimitDegrees=45, twistLimitDegrees=30}
 {bone1Name=dyn_cape_mid_05, bone2Name=dyn_cape_mid_04, swing1LimitDegrees=45, swing2LimitDegrees=45, twistLimitDegrees=30}
 {bone1Name=dyn_cape_left_parent, bone2Name=dyn_cape_left_base, swing1LimitDegrees=10, swing2LimitDegrees=10, twistLimitDegrees=5}
 {bone1Name=dyn_cape_left_01, bone2Name=dyn_cape_left_parent, swing1LimitDegrees=5, swing2LimitDegrees=5, twistLimitDegrees=5}
 {bone1Name=dyn_cape_left_02, bone2Name=dyn_cape_left_01, swing1LimitDegrees=5, swing2LimitDegrees=5, twistLimitDegrees=3}
 {bone1Name=dyn_cape_left_03, bone2Name=dyn_cape_left_02, swing1LimitDegrees=5, swing2LimitDegrees=5, twistLimitDegrees=3}
 {bone1Name=dyn_cape_left_04, bone2Name=dyn_cape_left_03, swing1LimitDegrees=45, swing2LimitDegrees=45, twistLimitDegrees=5}
 {bone1Name=dyn_cape_left_05, bone2Name=dyn_cape_left_04, swing1LimitDegrees=45, swing2LimitDegrees=45, twistLimitDegrees=5}
 {bone1Name=dyn_cape_left_06, bone2Name=dyn_cape_left_05, swing1LimitDegrees=45, swing2LimitDegrees=45, twistLimitDegrees=5}
 {bone1Name=dyn_cape_left_07, bone2Name=dyn_cape_left_06, swing1LimitDegrees=45, swing2LimitDegrees=45, twistLimitDegrees=5}
 {bone1Name=dyn_cape_right_parent, bone2Name=dyn_cape_right_base, swing1LimitDegrees=10, swing2LimitDegrees=10, twistLimitDegrees=5}
 {bone1Name=dyn_cape_right_01, bone2Name=dyn_cape_right_parent, swing1LimitDegrees=5, swing2LimitDegrees=5, twistLimitDegrees=30}
 {bone1Name=dyn_cape_right_02, bone2Name=dyn_cape_right_01, swing1LimitDegrees=5, swing2LimitDegrees=5, twistLimitDegrees=3}
 {bone1Name=dyn_cape_right_03, bone2Name=dyn_cape_right_02, swing1LimitDegrees=5, swing2LimitDegrees=5, twistLimitDegrees=5}
 {bone1Name=dyn_cape_right_04, bone2Name=dyn_cape_right_03, swing1LimitDegrees=45, swing2LimitDegrees=45, twistLimitDegrees=30}
 {
  bone1Name=dyn_cape_right_03
  bone2Name=dyn_cape_mid_03
  swing1LimitDegrees=45
  swing1Motion=Free
  swing2LimitDegrees=45
  swing2Motion=Free
  twistLimitDegrees=45
  twistMotion=Free
 }
 {
  bone1Name=dyn_cape_left_04
  bone2Name=dyn_cape_mid_04
  swing1LimitDegrees=45
  swing1Motion=Free
  swing2LimitDegrees=45
  swing2Motion=Free
  twistLimitDegrees=45
  twistMotion=Free
 }
 {
  bone1Name=dyn_cape_right_03
  bone2Name=dyn_cape_mid_03
  swing1LimitDegrees=45
  swing1Motion=Free
  swing2LimitDegrees=45
  swing2Motion=Free
  twistLimitDegrees=45
  twistMotion=Free
 }
 {
  bone1Name=dyn_cape_right_04
  bone2Name=dyn_cape_mid_04
  swing1LimitDegrees=45
  swing1Motion=Free
  swing2LimitDegrees=45
  swing2Motion=Free
  twistLimitDegrees=45
  twistMotion=Free
 }
]
