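# Physics asset for the grenade pieces on the female medium Scholar Witch armor, set up for dangling motion against the body.
# Two grenade capsules are each constrained to a tiny anchor sphere, forming separate two-body pairs.
# The dyn_grenade_01 pair allows a wide swing, the right-side pair is much stiffer, and both limit twist tightly.
# Unconstrained pelvis and left thigh capsules supply body collision shapes.
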PhysicsAsset /Game/Characters/Player/Female/Medium/Bodies/F_MED_Scholar_Witch_Armor/Meshes/F_MED_Scholar_Witch_Armor_Grenade_Physics
bodies {
 dyn_grenade_01 [{shapeType=Capsule, length=3.5274, radius=3.1017, center=<4.4474 0 -0.8045>, rotation=<rot 90 0 0>}]
 dyn_grenade_base_r [{shapeType=Sphere, radius=0.5}]
 dyn_grenade_parent [{shapeType=Sphere, radius=0.5}]
 dyn_grenade_parent_r [{shapeType=Capsule, length=2.8282, radius=3.6824, center=<-0.3291 0 -3.2773>, rotation=<rot 90 0 0>}]
 pelvis [{shapeType=Capsule, length=11.3497, radius=10.4329, center=<4.484 0 -1.1155>, rotation=<rot 74.5371 179.9998 179.9998>}]
 thigh_l [{shapeType=Capsule, length=22.1383, radius=7.8617, center=<-14.0943 -0.0003 0.2524>, rotation=<rot 87.1253 180 180>}]
}
constraints [
 {bone1Name=dyn_grenade_01, bone2Name=dyn_grenade_parent, swing1LimitDegrees=40, swing2LimitDegrees=40, twistLimitDegrees=5}
 {bone1Name=dyn_grenade_parent_r, bone2Name=dyn_grenade_base_r, swing1LimitDegrees=10, swing2LimitDegrees=10, twistLimitDegrees=5}
]
