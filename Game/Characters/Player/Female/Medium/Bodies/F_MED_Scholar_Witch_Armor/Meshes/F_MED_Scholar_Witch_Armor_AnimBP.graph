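# Post-copy Anim Blueprint for the female Scholar Witch Armor body, animating its shoulder pads, knife, pouch, grenades and physics bones.
# The copied parent pose gets shoulder-pad offsets driven by aiming, weapon type and parachute state, weighted bone constraints, then grenade RigidBody simulation.
# That cached pose feeds two RigidBody setups, selected by a one-time back-bling socket check.
# Skydiving adds external forces to both and resets dynamics on begin and end.
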AnimBlueprint /Game/Characters/Player/Female/Medium/Bodies/F_MED_Scholar_Witch_Armor/Meshes/F_MED_Scholar_Witch_Armor_AnimBP
target_skeleton=&/F_MED_Scholar_Witch_Armor_Skeleton
graphs [
 AnimGraph {nodes [
   n1 = RigidBody(
    ComponentPose=Local To Component(LocalPose=Use cached pose 'BackpackCache'())
    ExternalForce=Get Backpack_NotEquip_SkyDive_Velocity() -> (Backpack_NotEquip_SkyDive_Velocity:Vector)
   ) {
    bClampLinearTranslationLimitToRefPose=true
    bOverrideWorldGravity=true
    overridePhysicsAsset=&/F_MED_Scholar_Witch_Armor_Physics
    simulationSpace=BaseBoneSpace
    alphaBoolBlend {blendOption=Linear}
    alphaScaleBiasClamp {bInterpResult=true, interpSpeedDecreasing=6, interpSpeedIncreasing=8}
    baseBoneRef {boneName=root}
    componentAppliedLinearAccClamp=<3000 3000 3000>
    componentLinearAccScale=<0.25 0.25 0.35>
    componentLinearVelScale=<0.25 0.25 0.35>
    overrideWorldGravity {z=-980}
   }
   n2 = RigidBody(
    ComponentPose=Local To Component(LocalPose=Use cached pose 'BackpackCache'())
    ExternalForce=Get Backpack_Equip_SkyDive_Velocity() -> (Backpack_Equip_SkyDive_Velocity:Vector)
   ) {
    bClampLinearTranslationLimitToRefPose=true
    bOverrideWorldGravity=true
    overridePhysicsAsset=&/F_MED_Scholar_Witch_Armor_Backpack_Physics
    simulationSpace=BaseBoneSpace
    alphaBoolBlend {blendOption=Linear}
    alphaScaleBiasClamp {bInterpResult=true, interpSpeedDecreasing=6, interpSpeedIncreasing=8}
    baseBoneRef {boneName=root}
    componentAppliedLinearAccClamp=<3000 3000 3000>
    componentLinearAccScale=<0.25 0.25 0.35>
    componentLinearVelScale=<0.25 0.25 0.35>
    overrideWorldGravity {z=-980}
   }
   Output Pose(Result=Blend Poses by bool(
     BlendPose_0=Component To Local(ComponentPose=@n2)
     BlendPose_1=Component To Local(ComponentPose=@n1)
     bActiveValue=Get Is_Wearing_Backbling() -> (Is_Wearing_Backbling:bool)
    ) {blendType=Linear}) {layerGroup=None}
   n3 = Get bIsTargeting()
   n4 = Get ADS_Alpha() -> (ADS_Alpha:double)
   n5 = Get Skydive_ParachuteOpen_Alpha() -> (Skydive_ParachuteOpen_Alpha:double)
   n6 = "Transform (Modify) Bone - Bone: dyn_shoulderpad_r"(
    ComponentPose=Local To Component(LocalPose=Copy Pose From Mesh() {bUseAttachedParent=true})
    Alpha=@n5
    Translation="-3,2,-2.5"
    Rotation="20,12,-15"
    Scale="1,1,1"
   ) {
    rotationMode=BMM_Additive
    rotationSpace=BCS_BoneSpace
    translationMode=BMM_Additive
    translationSpace=BCS_ParentBoneSpace
    alphaBoolBlend {blendOption=Linear}
    alphaScaleBiasClamp {bInterpResult=true, interpSpeedDecreasing=1002, interpSpeedIncreasing=100}
    boneToModify {boneName=dyn_shoulderpad_r}
    rotation=<rot 20 12 -15>
    translation=<-3 2 -2.5>
   }
   n7 = "Transform (Modify) Bone - Bone: dyn_shoulderpad_l"(ComponentPose=@n6, Alpha=@n5, Translation="3,0,1.5", Rotation="12,-2,15", Scale="1,1,1") {
    rotationMode=BMM_Additive
    rotationSpace=BCS_BoneSpace
    translationMode=BMM_Additive
    translationSpace=BCS_ParentBoneSpace
    alphaBoolBlend {blendOption=Linear}
    alphaScaleBiasClamp {bInterpResult=true, interpSpeedDecreasing=1002, interpSpeedIncreasing=100}
    boneToModify {boneName=dyn_shoulderpad_l}
    rotation=<rot 12 -2 15>
    translation {x=3, z=1.5}
   }
   n8 = "Transform (Modify) Bone - Bone: dyn_shoulderpad_r"(ComponentPose=@n7, Alpha=@n4, Translation="0,0.5,0", Rotation="7.644,-3.039,27.978", Scale="1,1,1") {
    rotationMode=BMM_Additive
    rotationSpace=BCS_BoneSpace
    translationMode=BMM_Additive
    translationSpace=BCS_ParentBoneSpace
    alphaBoolBlend {blendOption=Linear}
    alphaScaleBiasClamp {bInterpResult=true, interpSpeedDecreasing=1, interpSpeedIncreasing=3}
    boneToModify {boneName=dyn_shoulderpad_r}
    rotation=<rot 7.644 -3.039 27.978>
    translation {y=0.5}
   }
   n9 = "Transform (Modify) Bone - Bone: dyn_shoulderpad_l"(ComponentPose=@n8, Alpha=@n4, Translation="0,0.184,2", Rotation="14.905,-1.251,9.814", Scale="1,1,1") {
    rotationMode=BMM_Additive
    rotationSpace=BCS_BoneSpace
    translationMode=BMM_Additive
    translationSpace=BCS_ParentBoneSpace
    alphaBoolBlend {blendOption=Linear}
    alphaScaleBiasClamp {bInterpResult=true, interpSpeedDecreasing=4, interpSpeedIncreasing=6}
    boneToModify {boneName=dyn_shoulderpad_l}
    rotation=<rot 14.905 -1.251 9.814>
    translation {y=0.184, z=2}
   }
   n10 = "Transform (Modify) Bone - Bone: dyn_shoulderpad_r"(
    ComponentPose=@n9
    Alpha=Get Weapon_Equipped() -> (Weapon_Equipped:double)
    Translation="0,0,-0.5"
    Rotation="-0.000,0.000,-20"
    Scale="1,1,1"
   ) {
    rotationMode=BMM_Additive
    rotationSpace=BCS_BoneSpace
    translationMode=BMM_Additive
    translationSpace=BCS_ParentBoneSpace
    alphaBoolBlend {blendOption=Linear}
    boneToModify {boneName=dyn_shoulderpad_r}
    rotation {roll=-20}
    translation {z=-0.5}
   }
   n11 = "Transform (Modify) Bone - Bone: dyn_shoulderpad_l"(
    ComponentPose=@n10
    Alpha="To Float (Boolean)"(InBool=NOT Boolean(A=@n3))
    Translation="0,0.184,2"
    Rotation="11.905,0.249,8.814"
    Scale="1,1,1"
   ) {
    rotationMode=BMM_Additive
    rotationSpace=BCS_BoneSpace
    translationMode=BMM_Additive
    translationSpace=BCS_ParentBoneSpace
    alphaBoolBlend {blendOption=Linear}
    alphaScaleBiasClamp {bInterpResult=true, interpSpeedDecreasing=4, interpSpeedIncreasing=6}
    boneToModify {boneName=dyn_shoulderpad_l}
    rotation=<rot 11.905 0.249 8.814>
    translation {y=0.184, z=2}
   }
   n12 = "Constraint - dyn_knife (PPPPP)"!(
    ComponentPose=@n11
    Alpha=1
    ConstraintWeights_0=0.0001
    ConstraintWeights_1:float=0.1154
    ConstraintWeights_2:float=0.5275
    ConstraintWeights_3:float=0.0073
    ConstraintWeights_4:float=0.3497
   ) -> (Pose) {
    alphaBoolBlend {blendOption=Linear}
    boneToModify {boneName=dyn_knife}
    constraintSetup [
     {targetBone {boneName=thigh_r}}
     {targetBone {boneName=pelvis}}
     {targetBone {boneName=thigh_twist_01_l}}
     {targetBone {boneName=thigh_twist_01_r}}
     {targetBone {boneName=thigh_l}}
    ]
    constraintWeights [1,1,1,1,1]
   }
   n13 = "Constraint - dyn_pouch (PPP)"!(ComponentPose=@n12, Alpha=1, ConstraintWeights_0=0.0516, ConstraintWeights_1:float=0.6013, ConstraintWeights_2:float=0.3472) -> (Pose) {
    alphaBoolBlend {blendOption=Linear}
    boneToModify {boneName=dyn_pouch}
    constraintSetup [{targetBone {boneName=thigh_r}}, {targetBone {boneName=pelvis}}, {targetBone {boneName=thigh_twist_01_r}}]
    constraintWeights [1,1,1]
   }
   n14 = "Constraint - dyn_grenade_base_r (PPPPP)"!(
    ComponentPose=@n13
    Alpha=1
    ConstraintWeights_0=0.2062
    ConstraintWeights_1:float=0.3388
    ConstraintWeights_2:float=0.042
    ConstraintWeights_3:float=0.0909
    ConstraintWeights_4:float=0.3222
   ) -> (Pose) {
    alphaBoolBlend {blendOption=Linear}
    boneToModify {boneName=dyn_grenade_base_r}
    constraintSetup [
     {targetBone {boneName=spine_01}}
     {targetBone {boneName=spine_02}}
     {targetBone {boneName=spine_03}}
     {targetBone {boneName=spine_04}}
     {targetBone {boneName=pelvis}}
    ]
    constraintWeights [1,1,1,1,1]
   }
   n15 = "Constraint - dyn_grenade_base (PPPPP)"!(
    ComponentPose=@n14
    Alpha=1
    ConstraintWeights_0=0.2038
    ConstraintWeights_1:float=0.3397
    ConstraintWeights_2:float=0.0534
    ConstraintWeights_3:float=0.1012
    ConstraintWeights_4:float=0.3019
   ) -> (Pose) {
    alphaBoolBlend {blendOption=Linear}
    boneToModify {boneName=dyn_grenade_base}
    constraintSetup [
     {targetBone {boneName=spine_01}}
     {targetBone {boneName=spine_02}}
     {targetBone {boneName=spine_03}}
     {targetBone {boneName=spine_04}}
     {targetBone {boneName=pelvis}}
    ]
    constraintWeights [1,1,1,1,1]
   }
   n16 = RigidBody(ComponentPose=@n15) {
    bClampLinearTranslationLimitToRefPose=true
    bOverrideWorldGravity=true
    overridePhysicsAsset=&/F_MED_Scholar_Witch_Armor_Grenade_Physics
    simulationSpace=BaseBoneSpace
    alphaBoolBlend {blendOption=Linear}
    alphaScaleBiasClamp {bInterpResult=true, interpSpeedDecreasing=4, interpSpeedIncreasing=6}
    baseBoneRef {boneName=pelvis}
    componentAppliedLinearAccClamp=<6000 6000 6000>
    componentLinearAccScale=<0.25 0.25 0.35>
    componentLinearVelScale=<0.25 0.25 0.35>
    overrideWorldGravity {z=-980}
   }
   Save cached pose 'BackpackCache'(Pose=Component To Local(ComponentPose=@n16)) {cacheName=BackpackCache}
  ]}
 EventGraph {nodes [
   Event BlueprintUpdateAnimation()
   > n1 = Sequence!(execute) -> (then_0, then_1, then_2:Exec, then_3:Exec, then_4:Exec, then_5:Exec)
   > Set ADS_Alpha(ADS_Alpha:double=SelectFloat(A=1, B=0, bPickA=Get bIsTargeting())) -> (Output_Get:double)
   Do Once(execute=@n1.then_1)
   n2 = GetPartSkeletalMeshComponent()
   n3 = FindChildCommponentFromAttachSocket!(
    self:Custom Character Part Anim Instance Object Reference
    Parent:Skeletal Mesh Component Object Reference=@n2
    bIncludeAllDescendants:bool=false
    AttachSocket:Name=cape
   ) -> (ReturnValue:Skeletal Mesh Component Object Reference)
   n4 = FindChildCommponentFromAttachSocket(Parent=@n2, bIncludeAllDescendants=false, AttachSocket=Backpack_BR)
   > Set Is_Wearing_Backbling(Is_Wearing_Backbling:bool=OR Boolean(A="Not Equal (Object)"!(A:Object Reference=@n3, B:Object Reference) -> (ReturnValue:bool), B="Not Equal (Object)"(A=@n4))) -> (Output_Get:bool)
   n5 = Cast To FortPlayerAnimInstance(execute=@n1.then_5, Object=GetAnimInstance(self=GetOwnerSkeletalMeshComponent()))
   > Set Weapon_Equipped(Weapon_Equipped:double=0) -> (Output_Get:double)
   > n6 = Switch on EFortWeaponCoreAnimation!(execute, Selection:EFortWeaponCoreAnimation Enum=Get WeaponCoreAnim(self=@n5)) -> (
    Melee:Exec
    Pistol:Exec
    Shotgun:Exec
    PaperBlueprint:Exec
    Rifle:Exec
    MeleeOneHand:Exec
    MachinePistol:Exec
    RocketLauncher:Exec
    GrenadeLauncher:Exec
    GoingCommando:Exec
    AssaultRifle:Exec
    TacticalShotgun:Exec
    SniperRifle:Exec
    TrapPlacement:Exec
    ShoulderLauncher:Exec
    AbilityDecoTool:Exec
    Crossbow:Exec
    C4:Exec
    RemoteControl:Exec
    DualWield:Exec
    AR_BullPup:Exec
    AR_ForwardGrip:Exec
    MedPackPaddles:Exec
    SMG_P90:Exec
    AR_DrumGun:Exec
    Consumable_Small:Exec
    Consumable_Large:Exec
    Balloon:Exec
    MountedTurret:Exec
    CreativeTool:Exec
    ExplosiveBow:Exec
    AshtonIndigo:Exec
    AshtonChicago:Exec
    MeleeDualWield:Exec
    Unarmed:Exec
   )
   Set Weapon_Equipped(execute=[@n6.Melee, @n6.Pistol], Weapon_Equipped=1)
   n7 = OR Boolean(A=Get bIsSkydiveFloating(), B=Get bIsSkydiveDiveMode())
   Set Backpack_Equip_SkyDive_Velocity(execute=@n1.then_2, Backpack_Equip_SkyDive_Velocity:Vector=SelectVector(A="0,0,20000", B="0, 0, 0", bPickA=@n7)) -> (Output_Get:Vector)
   Set Backpack_NotEquip_SkyDive_Velocity(execute=@n1.then_3, Backpack_NotEquip_SkyDive_Velocity:Vector=SelectVector(A="20000,20000,20000", B="0, 0, 0", bPickA=@n7)) -> (Output_Get:Vector)
   n8 = Get bIsSkyDiving()
   n9 = Get bIsParachuteOpened()
   Set Skydive_ParachuteOpen_Alpha(execute=@n1.then_4, Skydive_ParachuteOpen_Alpha:double=SelectFloat(A=1, B=0, bPickA=AND Boolean(A=@n9, B=@n8))) -> (Output_Get:double)
   n10 = Event OnBeginSkydiving()
   ResetDynamics(execute=[@n10.then, @n11.then], InTeleportType=ResetPhysics)
   n11 = Event OnEndSkydiving()
  ]}
]
variables [
 ADS_Alpha {type=number}
 Backpack_Equip_SkyDive_Velocity {type=Vector}
 Backpack_NotEquip_SkyDive_Velocity {type=Vector}
 Is_Wearing_Backbling {type=boolean}
 Skydive_ParachuteOpen_Alpha {type=number}
 Weapon_Equipped {type=number}
]
